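# Anim Blueprint for the Thrive backpack cosmetic, animating its five spike tips and hiding an element while the matching pickaxe is equipped, except during skydiving.
# AnimDynamics chains are selected by targeting and front-end state, then layered with random-vector drift (off when targeting) and fixed offsets.
# Spine bend relative to ref pose drives root roll and translation.
# Customization, weapon and skydive events set the "Hide Element 01" material parameter.
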AnimBlueprint /BRCosmetics/Accessories/FORT_Backpacks/Backpack_F_MED_Thrive/Meshes/F_MED_Thrive_Pack_AnimBP
target_skeleton=&/F_MED_Thrive_Pack_Skeleton
graphs [
 AnimGraph {nodes [
   Input Pose()
   Local Space Ref Pose()
   n1 = Anim Dynamics - Bone: spike_top_tip(ComponentPose=Local To Component(LocalPose=Local Space Ref Pose()), Alpha=0.4, ExternalForce="0,0,0") {
    angularDampingOverride=0.9
    angularSpringConstant=150
    bAngularSpring=true
    bGravityOverrideInSimSpace=true
    bLinearSpring=true
    bOverrideAngularDamping=true
    bOverrideLinearDamping=true
    gravityScale=0.05
    linearDampingOverride=0.995
    linearSpringConstant=150
    simulationSpace=RootRelative
    alphaBoolBlend {blendOption=Linear}
    alphaScaleBiasClamp {bInterpResult=true, clampMax=0.55, interpSpeedDecreasing=7, interpSpeedIncreasing=5}
    boundBone {boneName=spike_top_tip}
    componentAppliedLinearAccClamp=<6000 6000 6000>
    componentLinearAccScale=<0.15 0.15 0.15>
    componentLinearVelScale=<0.6 0.6 0.6>
    gravityOverride {z=-980}
    physicsBodyDefinitions [{
      sphereCollisionRadius=0
      boundBone {boneName=spike_top_tip}
      boxExtents=<100 100 100>
      constraintSetup {angularLimitsMax=<0 0 0>, angularLimitsMin=<0 0 0>, angularTarget=<0 0 0>, linearAxesMax=<10 2 0.5>, linearAxesMin=<-1 -3 -0.5>}
      localJointOffset=<0 0 0>
     }]
    relativeSpaceBone {boneName=root}
   }
   n2 = Anim Dynamics - Bone: spike_rightmid_tip(ComponentPose=@n1, Alpha=0.7, ExternalForce="0,0,0") {
    angularDampingOverride=0.9
    angularSpringConstant=150
    bAngularSpring=true
    bGravityOverrideInSimSpace=true
    bLinearSpring=true
    bOverrideAngularDamping=true
    bOverrideLinearDamping=true
    gravityScale=0.001
    linearDampingOverride=0.995
    linearSpringConstant=150
    simulationSpace=RootRelative
    alphaBoolBlend {blendOption=Linear}
    alphaScaleBiasClamp {bInterpResult=true, clampMax=0.5, interpSpeedDecreasing=7, interpSpeedIncreasing=5}
    boundBone {boneName=spike_rightmid_tip}
    componentAppliedLinearAccClamp=<6000 6000 6000>
    componentLinearAccScale=<0.15 0.15 0.15>
    componentLinearVelScale=<0.45 0.45 0.45>
    gravityOverride {z=-980}
    physicsBodyDefinitions [{
      sphereCollisionRadius=0
      boundBone {boneName=spike_rightmid_tip}
      boxExtents=<100 100 100>
      constraintSetup {angularLimitsMax=<0 0 0>, angularLimitsMin=<0 0 0>, angularTarget=<0 0 0>, linearAxesMax=<1 2 2>, linearAxesMin=<-0.5 -3 -6>}
      localJointOffset=<2 0 0>
     }]
    relativeSpaceBone {boneName=root}
   }
   n3 = Anim Dynamics - Bone: spike_rightbot_tip(ComponentPose=@n2, Alpha=0.7, ExternalForce="0,0,0") {
    angularDampingOverride=0.9
    angularSpringConstant=150
    bAngularSpring=true
    bGravityOverrideInSimSpace=true
    bLinearSpring=true
    bOverrideAngularDamping=true
    bOverrideLinearDamping=true
    gravityScale=0.02
    linearDampingOverride=0.995
    linearSpringConstant=150
    simulationSpace=RootRelative
    alphaBoolBlend {blendOption=Linear}
    alphaScaleBiasClamp {bInterpResult=true, clampMax=0.25, interpSpeedDecreasing=7, interpSpeedIncreasing=5}
    boundBone {boneName=spike_rightbot_tip}
    componentAppliedLinearAccClamp=<6000 6000 6000>
    componentLinearAccScale=<0.15 0.15 0.15>
    componentLinearVelScale=<0.45 0.45 0.45>
    gravityOverride {z=-980}
    physicsBodyDefinitions [{
      sphereCollisionRadius=0
      boundBone {boneName=spike_rightbot_tip}
      boxExtents=<100 100 100>
      constraintSetup {angularLimitsMax=<0 0 0>, angularLimitsMin=<0 0 0>, angularTarget=<0 0 0>, linearAxesMax=<1 2 2>, linearAxesMin=<-0.5 -3 -6>}
      localJointOffset=<2 0 0>
     }]
    relativeSpaceBone {boneName=root}
   }
   n4 = Anim Dynamics - Bone: spike_leftmid_tip(ComponentPose=@n3, Alpha=0.5, ExternalForce="0,0,0") {
    angularDampingOverride=0.9
    angularSpringConstant=150
    bAngularSpring=true
    bGravityOverrideInSimSpace=true
    bLinearSpring=true
    bOverrideAngularDamping=true
    bOverrideLinearDamping=true
    gravityScale=0.05
    linearDampingOverride=0.995
    linearSpringConstant=150
    simulationSpace=RootRelative
    alphaBoolBlend {blendOption=Linear}
    alphaScaleBiasClamp {bInterpResult=true, clampMax=0.3, interpSpeedDecreasing=7, interpSpeedIncreasing=5}
    boundBone {boneName=spike_leftmid_tip}
    componentAppliedLinearAccClamp=<6000 6000 6000>
    componentLinearAccScale=<0.15 0.15 0.15>
    componentLinearVelScale=<0.45 0.45 0.45>
    gravityOverride {z=-980}
    physicsBodyDefinitions [{
      sphereCollisionRadius=0
      boundBone {boneName=spike_leftmid_tip}
      boxExtents=<100 100 100>
      constraintSetup {angularLimitsMax=<0 0 0>, angularLimitsMin=<0 0 0>, angularTarget=<0 0 0>, linearAxesMax=<1 2 6>, linearAxesMin=<-0.5 -3 -2>}
      localJointOffset=<2 0 0>
     }]
    relativeSpaceBone {boneName=root}
   }
   n5 = Anim Dynamics - Bone: spike_leftbot_tip(ComponentPose=@n4, Alpha=0.8, ExternalForce="0,0,0") {
    angularDampingOverride=0.9
    angularSpringConstant=150
    bAngularSpring=true
    bGravityOverrideInSimSpace=true
    bLinearSpring=true
    bOverrideAngularDamping=true
    bOverrideLinearDamping=true
    gravityScale=0.05
    linearDampingOverride=0.995
    linearSpringConstant=150
    simulationSpace=RootRelative
    alphaBoolBlend {blendOption=Linear}
    alphaScaleBiasClamp {bInterpResult=true, clampMax=0.3, interpSpeedDecreasing=7, interpSpeedIncreasing=5}
    boundBone {boneName=spike_leftbot_tip}
    componentAppliedLinearAccClamp=<6000 6000 6000>
    componentLinearAccScale=<0.15 0.15 0.15>
    componentLinearVelScale=<0.45 0.45 0.45>
    gravityOverride {z=-980}
    physicsBodyDefinitions [{
      sphereCollisionRadius=0
      boundBone {boneName=spike_leftbot_tip}
      boxExtents=<100 100 100>
      constraintSetup {angularLimitsMax=<0 0 0>, angularLimitsMin=<0 0 0>, angularTarget=<0 0 0>, linearAxesMax=<1 2 6>, linearAxesMin=<-0.5 -3 -2>}
      localJointOffset=<2 0 0>
     }]
    relativeSpaceBone {boneName=root}
   }
   n6 = Anim Dynamics - Bone: spike_top_tip(ComponentPose=Local To Component(LocalPose=Local Space Ref Pose()), Alpha=1, ExternalForce="0,0,0") {
    angularDampingOverride=0.9
    angularSpringConstant=200000
    bGravityOverrideInSimSpace=true
    bLinearSpring=true
    bOverrideAngularDamping=true
    bOverrideLinearDamping=true
    gravityScale=0
    linearDampingOverride=0.995
    linearSpringConstant=75
    simulationSpace=BoneRelative
    alphaBoolBlend {blendOption=Linear}
    alphaScaleBiasClamp {bClampResult=true, bInterpResult=true, clampMax=0.3, interpSpeedDecreasing=7, interpSpeedIncreasing=5}
    boundBone {boneName=spike_top_tip}
    componentAppliedLinearAccClamp=<3000 3000 3000>
    componentLinearAccScale=<0.1 0.4 0.5>
    componentLinearVelScale=<0.7 0.7 0.7>
    gravityOverride {z=-980}
    physicsBodyDefinitions [{
      sphereCollisionRadius=0
      boundBone {boneName=spike_top_tip}
      boxExtents=<125 125 125>
      constraintSetup {angularLimitsMax=<0 0 0>, angularLimitsMin=<0 0 0>, angularTarget=<0 0 0>, linearAxesMax=<1 0.5 0.2>, linearAxesMin=<-1 -0.5 -0.2>}
      localJointOffset=<-0.1 0 -0.5>
     }]
    relativeSpaceBone {boneName=center}
   }
   n7 = Anim Dynamics - Bone: spike_rightmid_tip(ComponentPose=@n6, Alpha=1, ExternalForce="0,0,0") {
    angularDampingOverride=0.9
    angularSpringConstant=200000
    bGravityOverrideInSimSpace=true
    bLinearSpring=true
    bOverrideAngularDamping=true
    bOverrideLinearDamping=true
    gravityScale=0
    linearDampingOverride=0.995
    linearSpringConstant=75
    simulationSpace=BoneRelative
    alphaBoolBlend {blendOption=Linear}
    alphaScaleBiasClamp {bClampResult=true, bInterpResult=true, clampMax=0.3, interpSpeedDecreasing=7, interpSpeedIncreasing=5}
    boundBone {boneName=spike_rightmid_tip}
    componentAppliedLinearAccClamp=<3000 3000 3000>
    componentLinearAccScale=<0.1 0.4 0.5>
    componentLinearVelScale=<0.7 0.7 0.7>
    gravityOverride {z=-980}
    physicsBodyDefinitions [{
      sphereCollisionRadius=0
      boundBone {boneName=spike_rightmid_tip}
      boxExtents=<125 125 125>
      constraintSetup {angularLimitsMax=<0 0 0>, angularLimitsMin=<0 0 0>, angularTarget=<0 0 0>, linearAxesMax=<1 0.5 1>, linearAxesMin=<-0.5 -0.5 -0.5>}
      localJointOffset=<-0.1 0 -0.5>
     }]
    relativeSpaceBone {boneName=center}
   }
   n8 = Anim Dynamics - Bone: spike_rightbot_tip(ComponentPose=@n7, Alpha=1, ExternalForce="0,0,0") {
    angularDampingOverride=0.9
    angularSpringConstant=200000
    bGravityOverrideInSimSpace=true
    bLinearSpring=true
    bOverrideAngularDamping=true
    bOverrideLinearDamping=true
    gravityScale=0
    linearDampingOverride=0.995
    linearSpringConstant=75
    simulationSpace=BoneRelative
    alphaBoolBlend {blendOption=Linear}
    alphaScaleBiasClamp {bClampResult=true, bInterpResult=true, clampMax=0.3, interpSpeedDecreasing=7, interpSpeedIncreasing=5}
    boundBone {boneName=spike_rightbot_tip}
    componentAppliedLinearAccClamp=<3000 3000 3000>
    componentLinearAccScale=<0.1 0.4 0.5>
    componentLinearVelScale=<0.7 0.7 0.7>
    gravityOverride {z=-980}
    physicsBodyDefinitions [{
      sphereCollisionRadius=0
      boundBone {boneName=spike_rightbot_tip}
      boxExtents=<125 125 125>
      constraintSetup {angularLimitsMax=<0 0 0>, angularLimitsMin=<0 0 0>, angularTarget=<0 0 0>, linearAxesMax=<1 0.5 1>, linearAxesMin=<-0.5 -0.5 -0.5>}
      localJointOffset=<-0.1 0 -0.5>
     }]
    relativeSpaceBone {boneName=center}
   }
   n9 = Anim Dynamics - Bone: spike_leftmid_tip(ComponentPose=@n8, Alpha=1, ExternalForce="0,0,0") {
    angularDampingOverride=0.9
    angularSpringConstant=200000
    bGravityOverrideInSimSpace=true
    bLinearSpring=true
    bOverrideAngularDamping=true
    bOverrideLinearDamping=true
    gravityScale=0
    linearDampingOverride=0.995
    linearSpringConstant=75
    simulationSpace=BoneRelative
    alphaBoolBlend {blendOption=Linear}
    alphaScaleBiasClamp {bClampResult=true, bInterpResult=true, clampMax=0.3, interpSpeedDecreasing=7, interpSpeedIncreasing=5}
    boundBone {boneName=spike_leftmid_tip}
    componentAppliedLinearAccClamp=<3000 3000 3000>
    componentLinearAccScale=<0.1 0.4 0.5>
    componentLinearVelScale=<0.7 0.7 0.7>
    gravityOverride {z=-980}
    physicsBodyDefinitions [{
      sphereCollisionRadius=0
      boundBone {boneName=spike_leftmid_tip}
      boxExtents=<125 125 125>
      constraintSetup {angularLimitsMax=<0 0 0>, angularLimitsMin=<0 0 0>, angularTarget=<0 0 0>, linearAxesMax=<1 0.5 1>, linearAxesMin=<-0.5 -0.5 -0.5>}
      localJointOffset=<-0.1 0 -0.5>
     }]
    relativeSpaceBone {boneName=center}
   }
   n10 = Anim Dynamics - Bone: spike_leftbot_tip(ComponentPose=@n9, Alpha=1, ExternalForce="0,0,0") {
    angularDampingOverride=0.9
    angularSpringConstant=200000
    bGravityOverrideInSimSpace=true
    bLinearSpring=true
    bOverrideAngularDamping=true
    bOverrideLinearDamping=true
    gravityScale=0
    linearDampingOverride=0.995
    linearSpringConstant=75
    simulationSpace=BoneRelative
    alphaBoolBlend {blendOption=Linear}
    alphaScaleBiasClamp {bClampResult=true, bInterpResult=true, clampMax=0.3, interpSpeedDecreasing=7, interpSpeedIncreasing=5}
    boundBone {boneName=spike_leftbot_tip}
    componentAppliedLinearAccClamp=<3000 3000 3000>
    componentLinearAccScale=<0.1 0.4 0.5>
    componentLinearVelScale=<0.7 0.7 0.7>
    gravityOverride {z=-980}
    physicsBodyDefinitions [{
      sphereCollisionRadius=0
      boundBone {boneName=spike_leftbot_tip}
      boxExtents=<125 125 125>
      constraintSetup {angularLimitsMax=<0 0 0>, angularLimitsMin=<0 0 0>, angularTarget=<0 0 0>, linearAxesMax=<1 0.5 1>, linearAxesMin=<-0.5 -0.5 -0.5>}
      localJointOffset=<-0.1 0 -0.5>
     }]
    relativeSpaceBone {boneName=center}
   }
   n11 = Blend Poses by bool(BlendPose_0=Component To Local(ComponentPose=@n10), BlendPose_1=Component To Local(ComponentPose=@n5), bActiveValue=Get bIsInFrontEnd()) {blendType=Linear}
   n12 = Anim Dynamics - Bone: spike_top_tip(ComponentPose=Local To Component(LocalPose=Local Space Ref Pose()), Alpha=0.55, ExternalForce="0,0,0") {
    angularDampingOverride=0.9
    angularSpringConstant=200000
    bGravityOverrideInSimSpace=true
    bLinearSpring=true
    bOverrideAngularDamping=true
    bOverrideLinearDamping=true
    gravityScale=0
    linearDampingOverride=0.995
    linearSpringConstant=75
    simulationSpace=BoneRelative
    alphaBoolBlend {blendOption=Linear}
    alphaScaleBiasClamp {bClampResult=true, bInterpResult=true, clampMax=0.55, interpSpeedDecreasing=7, interpSpeedIncreasing=5}
    boundBone {boneName=spike_top_tip}
    componentAppliedLinearAccClamp=<3000 3000 3000>
    componentLinearAccScale=<0.1 0.4 0.5>
    componentLinearVelScale=<0.7 0.7 0.7>
    gravityOverride {z=-980}
    physicsBodyDefinitions [{
      sphereCollisionRadius=0
      boundBone {boneName=spike_top_tip}
      boxExtents=<125 125 125>
      constraintSetup {angularLimitsMax=<0 0 0>, angularLimitsMin=<0 0 0>, angularTarget=<0 0 0>, linearAxesMax=<1 0.5 0.2>, linearAxesMin=<-1 -0.5 -0.2>}
      localJointOffset=<-0.1 0 -0.5>
     }]
    relativeSpaceBone {boneName=center}
   }
   n13 = Anim Dynamics - Bone: spike_rightmid_tip(ComponentPose=@n12, Alpha=0.5, ExternalForce="0,0,0") {
    angularDampingOverride=0.9
    angularSpringConstant=200000
    bGravityOverrideInSimSpace=true
    bLinearSpring=true
    bOverrideAngularDamping=true
    bOverrideLinearDamping=true
    gravityScale=0
    linearDampingOverride=0.995
    linearSpringConstant=75
    simulationSpace=BoneRelative
    alphaBoolBlend {blendOption=Linear}
    alphaScaleBiasClamp {bClampResult=true, bInterpResult=true, clampMax=0.5, interpSpeedDecreasing=7, interpSpeedIncreasing=5}
    boundBone {boneName=spike_rightmid_tip}
    componentAppliedLinearAccClamp=<3000 3000 3000>
    componentLinearAccScale=<0.1 0.4 0.5>
    componentLinearVelScale=<0.7 0.7 0.7>
    gravityOverride {z=-980}
    physicsBodyDefinitions [{
      sphereCollisionRadius=0
      boundBone {boneName=spike_rightmid_tip}
      boxExtents=<125 125 125>
      constraintSetup {angularLimitsMax=<0 0 0>, angularLimitsMin=<0 0 0>, angularTarget=<0 0 0>, linearAxesMax=<1 0.5 1>, linearAxesMin=<-0.5 -0.5 -0.5>}
      localJointOffset=<-0.1 0 -0.5>
     }]
    relativeSpaceBone {boneName=center}
   }
   n14 = Anim Dynamics - Bone: spike_rightbot_tip(ComponentPose=@n13, Alpha=0.25, ExternalForce="0,0,0") {
    angularDampingOverride=0.9
    angularSpringConstant=200000
    bGravityOverrideInSimSpace=true
    bLinearSpring=true
    bOverrideAngularDamping=true
    bOverrideLinearDamping=true
    gravityScale=0
    linearDampingOverride=0.995
    linearSpringConstant=75
    simulationSpace=BoneRelative
    alphaBoolBlend {blendOption=Linear}
    alphaScaleBiasClamp {bClampResult=true, bInterpResult=true, clampMax=0.25, interpSpeedDecreasing=7, interpSpeedIncreasing=5}
    boundBone {boneName=spike_rightbot_tip}
    componentAppliedLinearAccClamp=<3000 3000 3000>
    componentLinearAccScale=<0.1 0.4 0.5>
    componentLinearVelScale=<0.7 0.7 0.7>
    gravityOverride {z=-980}
    physicsBodyDefinitions [{
      sphereCollisionRadius=0
      boundBone {boneName=spike_rightbot_tip}
      boxExtents=<125 125 125>
      constraintSetup {angularLimitsMax=<0 0 0>, angularLimitsMin=<0 0 0>, angularTarget=<0 0 0>, linearAxesMax=<1 0.5 1>, linearAxesMin=<-0.5 -0.5 -0.5>}
      localJointOffset=<-0.1 0 -0.5>
     }]
    relativeSpaceBone {boneName=center}
   }
   n15 = Anim Dynamics - Bone: spike_leftmid_tip(ComponentPose=@n14, Alpha=0.3, ExternalForce="0,0,0") {
    angularDampingOverride=0.9
    angularSpringConstant=200000
    bGravityOverrideInSimSpace=true
    bLinearSpring=true
    bOverrideAngularDamping=true
    bOverrideLinearDamping=true
    gravityScale=0
    linearDampingOverride=0.995
    linearSpringConstant=75
    simulationSpace=BoneRelative
    alphaBoolBlend {blendOption=Linear}
    alphaScaleBiasClamp {bClampResult=true, bInterpResult=true, clampMax=0.3, interpSpeedDecreasing=7, interpSpeedIncreasing=5}
    boundBone {boneName=spike_leftmid_tip}
    componentAppliedLinearAccClamp=<3000 3000 3000>
    componentLinearAccScale=<0.1 0.4 0.5>
    componentLinearVelScale=<0.7 0.7 0.7>
    gravityOverride {z=-980}
    physicsBodyDefinitions [{
      sphereCollisionRadius=0
      boundBone {boneName=spike_leftmid_tip}
      boxExtents=<125 125 125>
      constraintSetup {angularLimitsMax=<0 0 0>, angularLimitsMin=<0 0 0>, angularTarget=<0 0 0>, linearAxesMax=<1 0.5 1>, linearAxesMin=<-0.5 -0.5 -0.5>}
      localJointOffset=<-0.1 0 -0.5>
     }]
    relativeSpaceBone {boneName=center}
   }
   n16 = Anim Dynamics - Bone: spike_leftbot_tip(ComponentPose=@n15, Alpha=0.3, ExternalForce="0,0,0") {
    angularDampingOverride=0.9
    angularSpringConstant=200000
    bGravityOverrideInSimSpace=true
    bLinearSpring=true
    bOverrideAngularDamping=true
    bOverrideLinearDamping=true
    gravityScale=0
    linearDampingOverride=0.995
    linearSpringConstant=75
    simulationSpace=BoneRelative
    alphaBoolBlend {blendOption=Linear}
    alphaScaleBiasClamp {bClampResult=true, bInterpResult=true, clampMax=0.3, interpSpeedDecreasing=7, interpSpeedIncreasing=5}
    boundBone {boneName=spike_leftbot_tip}
    componentAppliedLinearAccClamp=<3000 3000 3000>
    componentLinearAccScale=<0.1 0.4 0.5>
    componentLinearVelScale=<0.7 0.7 0.7>
    gravityOverride {z=-980}
    physicsBodyDefinitions [{
      sphereCollisionRadius=0
      boundBone {boneName=spike_leftbot_tip}
      boxExtents=<125 125 125>
      constraintSetup {angularLimitsMax=<0 0 0>, angularLimitsMin=<0 0 0>, angularTarget=<0 0 0>, linearAxesMax=<1 0.5 1>, linearAxesMin=<-0.5 -0.5 -0.5>}
      localJointOffset=<-0.1 0 -0.5>
     }]
    relativeSpaceBone {boneName=center}
   }
   n17 = Component To Local(ComponentPose=@n16)
   n18 = Get bIsTargeting()
   n19 = Get Target_off() -> (Target_off:double)
   n20 = "Transform (Modify) Bone - Bone: spike_top_tip"!(
    ComponentPose=Local To Component(LocalPose=Blend Poses by bool(BlendPose_0=@n17, BlendPose_1=@n11, bActiveValue=@n18) {blendType=Linear})
    Alpha=@n19
    Translation=Get NewVar_0() -> (NewVar_0:Vector)
   ) -> (Pose) {
    translationMode=BMM_Additive
    translationSpace=BCS_BoneSpace
    alphaBoolBlend {blendOption=Linear}
    alphaScaleBiasClamp {bInterpResult=true, interpSpeedIncreasing=20}
    boneToModify {boneName=spike_top_tip}
   }
   n21 = "Transform (Modify) Bone - Bone: spike_rightmid_tip"!(ComponentPose=@n20, Alpha=@n19, Translation=Get NewVar_2() -> (NewVar_2:Vector)) -> (Pose) {
    translationMode=BMM_Additive
    translationSpace=BCS_BoneSpace
    alphaBoolBlend {blendOption=Linear}
    alphaScaleBiasClamp {bInterpResult=true, interpSpeedIncreasing=20}
    boneToModify {boneName=spike_rightmid_tip}
   }
   n22 = "Transform (Modify) Bone - Bone: spike_rightbot_tip"!(ComponentPose=@n21, Alpha=@n19, Translation=Get NewVar_1() -> (NewVar_1:Vector)) -> (Pose) {
    translationMode=BMM_Additive
    translationSpace=BCS_BoneSpace
    alphaBoolBlend {blendOption=Linear}
    alphaScaleBiasClamp {bInterpResult=true, interpSpeedIncreasing=20}
    boneToModify {boneName=spike_rightbot_tip}
   }
   n23 = "Transform (Modify) Bone - Bone: spike_leftmid_tip"!(ComponentPose=@n22, Alpha=@n19, Translation=Get NewVar_1()) -> (Pose) {
    translationMode=BMM_Additive
    translationSpace=BCS_BoneSpace
    alphaBoolBlend {blendOption=Linear}
    alphaScaleBiasClamp {bInterpResult=true, interpSpeedIncreasing=20}
    boneToModify {boneName=spike_leftmid_tip}
   }
   n24 = "Transform (Modify) Bone - Bone: spike_leftbot_tip"!(ComponentPose=@n23, Alpha=@n19, Translation=Get NewVar_3() -> (NewVar_3:Vector)) -> (Pose) {
    translationMode=BMM_Additive
    translationSpace=BCS_BoneSpace
    alphaBoolBlend {blendOption=Linear}
    alphaScaleBiasClamp {bInterpResult=true, interpSpeedIncreasing=20}
    boneToModify {boneName=spike_leftbot_tip}
   }
   n25 = "Transform (Modify) Bone - Bone: spike_top_tip"(ComponentPose=@n24, Alpha=1, Translation="-1.2,0.000,-0.000") {
    alpha=0
    translationMode=BMM_Additive
    translationSpace=BCS_BoneSpace
    alphaBoolBlend {blendOption=Linear}
    alphaScaleBiasClamp {bInterpResult=true, interpSpeedIncreasing=20}
    boneToModify {boneName=spike_top_tip}
    translation {x=-1.2}
   }
   n26 = "Transform (Modify) Bone - Bone: spike_rightmid_tip"(ComponentPose=@n25, Alpha=1, Translation="-1.000,-0.000,-0.213") {
    alpha=0
    translationMode=BMM_Additive
    translationSpace=BCS_BoneSpace
    alphaBoolBlend {blendOption=Linear}
    alphaScaleBiasClamp {bInterpResult=true, interpSpeedIncreasing=20}
    boneToModify {boneName=spike_rightmid_tip}
    translation {x=-1, z=-0.213}
   }
   n27 = "Transform (Modify) Bone - Bone: spike_rightbot_tip"(ComponentPose=@n26, Alpha=1, Translation="-0.5,0.000,-0.000") {
    alpha=0
    translationMode=BMM_Additive
    translationSpace=BCS_BoneSpace
    alphaBoolBlend {blendOption=Linear}
    alphaScaleBiasClamp {bInterpResult=true, interpSpeedIncreasing=20}
    boneToModify {boneName=spike_rightbot_tip}
    translation {x=-0.5}
   }
   n28 = "Transform (Modify) Bone - Bone: spike_leftmid_tip"(ComponentPose=@n27, Alpha=1, Translation="-0.800,0.000,-0.155") {
    alpha=0
    translationMode=BMM_Additive
    translationSpace=BCS_BoneSpace
    alphaBoolBlend {blendOption=Linear}
    alphaScaleBiasClamp {bInterpResult=true, interpSpeedIncreasing=20}
    boneToModify {boneName=spike_leftmid_tip}
    translation {x=-0.8, z=-0.155}
   }
   n29 = "Transform (Modify) Bone - Bone: spike_leftbot_tip"(ComponentPose=@n28, Alpha=1, Translation="-0.5,-0.000,0.000") {
    alpha=0
    translationMode=BMM_Additive
    translationSpace=BCS_BoneSpace
    alphaBoolBlend {blendOption=Linear}
    alphaScaleBiasClamp {bInterpResult=true, interpSpeedIncreasing=20}
    boneToModify {boneName=spike_leftbot_tip}
    translation {x=-0.5}
   }
   n30 = "Transform (Modify) Bone - Bone: root"(ComponentPose=@n29, Translation=Get backpack_translation(), Rotation=Get backpack_rotate()) {alpha=0, rotationMode=BMM_Additive, translationMode=BMM_Additive, alphaBoolBlend {blendOption=Linear}, boneToModify {boneName=root}}
   Output Pose(Result=Component To Local(ComponentPose=@n30)) {layerGroup=None}
  ]}
 EventGraph {nodes [
   Event BlueprintUpdateAnimation()
   > n1 = Sequence!(execute) -> (then_0, then_1, then_2:Exec, then_3:Exec, then_4:Exec, then_5:Exec, then_6:Exec)
   > n2 = Random Vector Function!(
    execute
    CharacterPartAnimInstance:Custom Character Part Anim Instance Object Reference
    SkelMeshComponent:Mesh Component Object Reference
    JointName:Name=None
    GameTimeInSeconds:double=0
    TimeOffset:double=Get customSpeed() -> (customSpeed:double)
    PerlinScale:double=0
    Frequency:Vector="0.7,0.5,0.7"
    Amplitude:Vector="0.5,0.35,0.3"
    Offset:Vector="0,0,0"
   ) -> (then, VectorOut:Vector)
   > Set NewVar_0(NewVar_0:Vector=@n2) -> (Output_Get:Vector)
   n3 = Random Vector Function(
    execute=@n1.then_1
    JointName=None
    GameTimeInSeconds=0
    TimeOffset=Get customSpeed()
    PerlinScale=0
    Frequency="0.8,0.8,0.4"
    Amplitude="0.4,0.3,0.3"
    Offset="0,0,0"
   )
   > Set NewVar_1(NewVar_1:Vector=@n3) -> (Output_Get:Vector)
   n4 = Random Vector Function(
    execute=@n1.then_2
    JointName=None
    GameTimeInSeconds=0
    TimeOffset=Get customSpeed()
    PerlinScale=0
    Frequency="0.5,0.3,0.5"
    Amplitude="0.4,0.2,0.2"
    Offset="0,0,0"
   )
   > Set NewVar_2(NewVar_2:Vector=@n4) -> (Output_Get:Vector)
   n5 = Random Vector Function(
    execute=@n1.then_3
    JointName=None
    GameTimeInSeconds=0
    TimeOffset=Get customSpeed()
    PerlinScale=0
    Frequency="0.6,0.5,0.4"
    Amplitude="0.5,0.2,0.2"
    Offset="0,0,0"
   )
   > Set NewVar_3(NewVar_3:Vector=@n5) -> (Output_Get:Vector)
   Set customSpeed(execute=@n1.then_4, customSpeed:double=float / float(A=GetRealTimeSeconds(), B=1.25) {is=title K2Node_CallFunction}) -> (Output_Get:double)
   n6 = Branch(execute=@n1.then_5, Condition=Get bIsTargeting())
   > Set Target_off(Target_off:double=0) -> (Output_Get:double)
   Set Target_off(execute=@n6.else, Target_off=1)
   n7 = GetOwnerSkeletalMeshComponent()
   n8 = BreakRotator(InRot=BreakTransform(InTransform=GetDeltaTransformFromRefPose(self=@n7, BoneName=spine_05, BaseName=spine_01)).Rotation)
   Set backpack_rotate(execute=@n1.then_6, backpack_rotate=MakeRotator(Roll=MapRangeClamped(Value=@n8.Yaw, InRangeA=0, InRangeB=-35, OutRangeB=40)))
   > Set backpack_translation(backpack_translation=MakeVector(
     Y=MapRangeClamped(Value=@n8.Yaw, InRangeA=0, InRangeB=-5, OutRangeB=-3)
     Z=MapRangeClamped(Value=@n8.Yaw, InRangeA=0, InRangeB=-6, OutRangeB=-4)
    ))
   n9 = Event OnCharacterCustomizationLoaded()
   n10 = Get Dynamic Material Instance(execute=[@n9.then, @n18.then], SkeletalMeshComponent=GetPartSkeletalMeshComponent())
   > n11 = Sequence!(execute) -> (then_0, then_1)
   n12 = PickaxeItemDefinitionHasAnyMetaTag(Tags=Make Array("[0]"=(TagName="Cosmetics.MetaTags.ThriveFemale")), bRequiresEquipped=true)
   > n13 = Branch(Condition=@n12)
   > n14 = Branch(Condition=Get bIsSkyDiving())
   SetScalarParameterValue(execute=[@n16.then, @n14.then, @n15.else, @n13.else], self=@n10, ParameterName=Hide Element 01, Value=0)
   SetScalarParameterValue(execute=[@n14.else, @n15.then], self=@n10, ParameterName=Hide Element 01, Value=1)
   n15 = Branch(execute=[@n17.then, @n11.then_1], Condition=@n12)
   n16 = Event OnBeginSkydiving()
   n17 = Event OnEndSkydiving()
   n18 = Event OnWeaponChanged()
   TryGetPawnOwner()
  ]}
]
variables [
 backpack_rotate {type=Rotator}
 backpack_translation {type=Vector}
 customSpeed {type=number}
 NewVar_0 {type=Vector}
 NewVar_1 {type=Vector}
 NewVar_2 {type=Vector}
 NewVar_3 {type=Vector}
 Target_off {type=number}
]
